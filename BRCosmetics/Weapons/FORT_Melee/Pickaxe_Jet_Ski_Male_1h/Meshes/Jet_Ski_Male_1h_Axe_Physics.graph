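# Physics asset for the Jet Ski Male one-handed pickaxe cosmetic, setting up simulation bodies for the weapon and its dangling ribbon.
# Three box bodies cover the handle bone, a tiny ribbon master bone and the ribbon itself, constrained in a chain.
# The master rotates freely on all angular axes relative to the handle, while the ribbon has wide swing limits and a very tight twist limit relative to the master.
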PhysicsAsset /BRCosmetics/Weapons/FORT_Melee/Pickaxe_Jet_Ski_Male_1h/Meshes/Jet_Ski_Male_1h_Axe_Physics
bodies {
 Melee_TwoHanded_Handle [{shapeType=Box, extentX=48.3874, extentY=7.3329, extentZ=33.3885, center=<-5.441 -7.7282 47.2528>}]
 dyn_ribbon [{shapeType=Box, extentX=8.3786, extentY=1.6339, extentZ=2.2679, center=<4.968 -0.537 0.2449>}]
 dyn_ribbon_mstr [{shapeType=Box, extentX=1.01, extentY=1.01, extentZ=1.01}]
}
constraints [
 {
  bone1Name=dyn_ribbon_mstr
  bone2Name=Melee_TwoHanded_Handle
  swing1LimitDegrees=45
  swing1Motion=Free
  swing2LimitDegrees=45
  swing2Motion=Free
  twistLimitDegrees=45
  twistMotion=Free
 }
 {bone1Name=dyn_ribbon, bone2Name=dyn_ribbon_mstr, swing1LimitDegrees=90, swing2LimitDegrees=90, twistLimitDegrees=5}
]
